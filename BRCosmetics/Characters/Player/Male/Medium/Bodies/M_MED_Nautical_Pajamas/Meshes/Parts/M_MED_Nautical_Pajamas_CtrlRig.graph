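# Control Rig for the M_MED_Nautical_Pajamas player cosmetic that shifts the upper mask bones in response to cheek movement.
# It reads each inner cheek bone's Z translation relative to faceAttach and remaps it to a clamped weight.
# That weight drives an additive local offset on dyn_mask_up_l, then dyn_mask_up_r.
# Two rotator-to-quaternion nodes are unconnected, and the wing, eye and hood bones are not driven.
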ControlRigBlueprint /BRCosmetics/Characters/Player/Male/Medium/Bodies/M_MED_Nautical_Pajamas/Meshes/Parts/M_MED_Nautical_Pajamas_CtrlRig
graphs [
 RigVMFunctionLibrary {links [], nodes []}
 RigVMModel {nodes [
   n1 = &RigUnit_BeginExecution()
   /Script/ControlRig.RigUnit_ModifyBoneTransforms(
    BoneToModify=((Bone="dyn_mask_up_l",Transform=(Rotation=(X=0,Y=0,Z=0,W=1),Translation=(X=1.2,Y=0,Z=0.2),Scale3D=(X=1,Y=1,Z=1))))
    Mode=AdditiveLocal
    WeightMaximum=1
    WeightMinimum=0
    ExecutePin=@n1.ExecutePin
    Weight=&RigVMFunction_MathFloatRemap(
     SourceMaximum=2
     SourceMinimum=0.11
     TargetMinimum=0
     bClamp=true
     Value=&RigUnit_GetRelativeBoneTransform(Bone=L_cheek_inner, Space=faceAttach).Transform.Translation.Z
    ).Result
   ) {is=struct RigVMUnitNode}
   > &RigUnit_ModifyBoneTransforms(
    BoneToModify=((Bone="dyn_mask_up_r",Transform=(Rotation=(X=0,Y=0,Z=0,W=1),Translation=(X=1.2,Y=0,Z=0.2),Scale3D=(X=1,Y=1,Z=1))))
    Mode=AdditiveLocal
    WeightMaximum=1
    WeightMinimum=0
    Weight=&RigVMFunction_MathFloatRemap(
     SourceMaximum=2
     SourceMinimum=0.11
     TargetMinimum=0
     bClamp=true
     Value=&RigUnit_GetRelativeBoneTransform(Bone=R_cheek_inner, Space=faceAttach).Transform.Translation.Z
    ).Result
   )
   /Script/RigVM.RigVMFunction_MathQuaternionFromRotator(Rotator=(Pitch=1,Yaw=0,Roll=-0.5)) {is=struct RigVMUnitNode}
   &RigVMFunction_MathQuaternionFromRotator(Rotator=(Pitch=-1,Yaw=0,Roll=0.5))
  ]}
]
hierarchy [
 root {loc=<0 0 0>, rot=<0 0 0 1>}
 pelvis {parent=root, loc=<0 -0.7071 93.1191>, rot=<0.0127 0.707 -0.0127 -0.707>}
 spine_01 {parent=pelvis, loc=<6.7494 -0.6388 0>, rot=<0 0 0.034 -0.9994>}
 spine_02 {parent=spine_01, loc=<7.2518 -0.035 0>, rot=<0 0 0.0238 -0.9997>}
 spine_03 {parent=spine_02, loc=<6.0104 0 0>, rot=<0 0 0.0149 -0.9999>}
 spine_04 {parent=spine_03, loc=<8.1024 0 0>, rot=<0 0 -0.0132 -0.9999>}
 spine_05 {parent=spine_04, loc=<10.0146 0 0>, rot=<0 0 -0.0015 -1>}
 neck_01 {parent=spine_05, loc=<11.3138 0.1529 0>, rot=<0 0 -0.2003 -0.9797>}
 neck_02 {parent=neck_01, loc=<4.5627 0 0>, rot=<0 0 0.0504 -0.9987>}
 head {parent=neck_02, loc=<5.3914 0 0>, rot=<0 0 0.115 -0.9934>}
 faceAttach {parent=head, loc=<0.6201 0.9296 0>, rot=<0.0082 -0.7071 -0.0082 -0.7071>, scale=<0.89 0.89 0.89>}
 L_cheek_inner {parent=faceAttach, loc=<4.2335 7.7652 0.1106>, rot=<-0.0672 -0.2612 0.3196 -0.9084>}
 C_nose_bridge {parent=faceAttach, loc=<-0.0001 11.1254 1.8962>, rot=<-0.2218 0.0035 0.001 -0.9751>}
 R_cheek_inner {parent=faceAttach, loc=<-4.2335 7.7652 0.1106>, rot=<-0.0681 0.2634 -0.3193 -0.9078>}
 C_jaw {parent=faceAttach, loc=<-1.075 2.2203 -1.795>, rot=<0 0 0 1>}
 dyn_wings_r_MSTR {parent=head, loc=<11.5183 -3.6958 6.4929>, rot=<0.2759 0.3574 0.404 -0.7956>}
 dyn_wings_01_r {parent=dyn_wings_r_MSTR, loc=<2.8162 0 0>, rot=<0 0 0 1>}
 dyn_wings_02_r {parent=dyn_wings_01_r, loc=<8.3661 7.9579 -0.1321>, rot=<0.4653 -0.5277 0.4482 -0.5515>}
 dyn_eyes_r_MSTR {parent=head, loc=<10.1664 4.1729 7.046>, rot=<0.6095 0.3926 0.2332 -0.6481>}
 dyn_eyes_r {parent=dyn_eyes_r_MSTR, loc=<0 0 0>, rot=<0 0 0 1>}
 dyn_wings_l_MSTR {parent=head, loc=<11.5187 -3.6958 -6.4929>, rot=<0.7956 -0.404 -0.3574 -0.2759>}
 dyn_wings_01_l {parent=dyn_wings_l_MSTR, loc=<2.8161 0.0001 0>, rot=<0 0 0 1>}
 dyn_wings_02_l {parent=dyn_wings_01_l, loc=<8.3661 -7.9578 -0.1321>, rot=<-0.4813 -0.5073 -0.431 -0.5703>}
 dyn_eyes_l_MSTR {parent=head, loc=<10.1666 4.1729 -7.046>, rot=<0.6481 -0.2332 -0.3926 -0.6095>}
 dyn_eyes_l {parent=dyn_eyes_l_MSTR, loc=<0 0 0>, rot=<0 0 0 -1>}
 dyn_hood_r_MSTR {parent=head, loc=<3.2327 -5.0469 9.0892>, rot=<-0.1113 0.1104 -0.7012 -0.6955>}
 dyn_hood_r {parent=dyn_hood_r_MSTR, loc=<0 0 0>, rot=<0 0 0 1>}
 dyn_hood_l_MSTR {parent=head, loc=<3.2326 -5.0469 -9.0892>, rot=<-0.6955 -0.7012 0.1104 -0.1113>}
 dyn_hood_l {parent=dyn_hood_l_MSTR, loc=<0 0 0>, rot=<0 0 0 1>}
 dyn_mask_up_l {parent=head, loc=<1.7754 8.957 -3.519>, rot=<-0.2935 -0.6433 -0.2988 -0.6409>}
 dyn_mask_up_r {parent=head, loc=<1.7748 8.957 3.519>, rot=<0.6409 0.2988 0.6433 0.2935>}
]
