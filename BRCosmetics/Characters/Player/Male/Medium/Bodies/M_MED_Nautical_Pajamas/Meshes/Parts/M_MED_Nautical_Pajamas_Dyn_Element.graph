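# Physics asset for the dynamic element mesh part of the male medium Nautical Pajamas body cosmetic, covering left and right dyn_eyes and dyn_wings bones around the head.
# Capsule bodies form four chains from a large head body, each starting at a tiny MSTR anchor with free swing and twist.
# Each eye is one body with tight limits; each wing has two segments, the outer limited more loosely than the inner.
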PhysicsAsset /BRCosmetics/Characters/Player/Male/Medium/Bodies/M_MED_Nautical_Pajamas/Meshes/Parts/M_MED_Nautical_Pajamas_Dyn_Element
bodies {
 dyn_eyes_l [{shapeType=Capsule, length=2, radius=3.5, center=<3 0 0>, rotation=<rot 90 0 0>}]
 dyn_eyes_l_MSTR [{shapeType=Capsule, length=0.5, radius=0.5}]
 dyn_eyes_r [{shapeType=Capsule, length=2, radius=3.5, center=<3 0 0>, rotation=<rot 90 0 0>}]
 dyn_eyes_r_MSTR [{shapeType=Capsule, length=0.5, radius=0.5}]
 dyn_wings_01_l [{shapeType=Capsule, length=1, radius=1, center=<-1.4 2.5 0>, rotation=<rot 90 0 0>}]
 dyn_wings_01_r [{shapeType=Capsule, length=1, radius=1, center=<-1.4 -2.5 0>, rotation=<rot 90 0 0>}]
 dyn_wings_02_l [{shapeType=Capsule, length=3, radius=2.25, center=<5 0 0>, rotation=<rot 90 0 0>}]
 dyn_wings_02_r [{shapeType=Capsule, length=3, radius=2.25, center=<5 0 0>, rotation=<rot 90 0 0>}]
 dyn_wings_l_MSTR [{shapeType=Capsule, length=0.5, radius=0.5}]
 dyn_wings_r_MSTR [{shapeType=Capsule, length=0.5, radius=0.5}]
 head [{shapeType=Capsule, length=4.7645, radius=11.2817, center=<4.3195 0.4384 -0.0001>, rotation=<rot 90 -14.0363 -34.0392>}]
}
constraints [
 {
  bone1Name=dyn_eyes_l_MSTR
  bone2Name=head
  swing1LimitDegrees=45
  swing1Motion=Free
  swing2LimitDegrees=45
  swing2Motion=Free
  twistLimitDegrees=45
  twistMotion=Free
 }
 {bone1Name=dyn_eyes_l, bone2Name=dyn_eyes_l_MSTR, swing1LimitDegrees=20, swing2LimitDegrees=20, twistLimitDegrees=2}
 {
  bone1Name=dyn_eyes_r_MSTR
  bone2Name=head
  swing1LimitDegrees=45
  swing1Motion=Free
  swing2LimitDegrees=45
  swing2Motion=Free
  twistLimitDegrees=45
  twistMotion=Free
 }
 {bone1Name=dyn_eyes_r, bone2Name=dyn_eyes_r_MSTR, swing1LimitDegrees=20, swing2LimitDegrees=20, twistLimitDegrees=2}
 {
  bone1Name=dyn_wings_r_MSTR
  bone2Name=head
  swing1LimitDegrees=45
  swing1Motion=Free
  swing2LimitDegrees=45
  swing2Motion=Free
  twistLimitDegrees=45
  twistMotion=Free
 }
 {bone1Name=dyn_wings_01_r, bone2Name=dyn_wings_r_MSTR, swing1LimitDegrees=10, swing2LimitDegrees=10, twistLimitDegrees=2}
 {bone1Name=dyn_wings_02_r, bone2Name=dyn_wings_01_r, swing1LimitDegrees=30, swing2LimitDegrees=30, twistLimitDegrees=15}
 {
  bone1Name=dyn_wings_l_MSTR
  bone2Name=head
  swing1LimitDegrees=45
  swing1Motion=Free
  swing2LimitDegrees=45
  swing2Motion=Free
  twistLimitDegrees=45
  twistMotion=Free
 }
 {bone1Name=dyn_wings_01_l, bone2Name=dyn_wings_l_MSTR, swing1LimitDegrees=10, swing2LimitDegrees=10, twistLimitDegrees=2}
 {bone1Name=dyn_wings_02_l, bone2Name=dyn_wings_01_l, swing1LimitDegrees=30, swing2LimitDegrees=30, twistLimitDegrees=15}
]
