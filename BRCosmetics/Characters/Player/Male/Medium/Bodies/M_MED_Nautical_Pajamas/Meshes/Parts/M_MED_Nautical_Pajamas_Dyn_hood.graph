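# Physics asset for the left and right hood bones on the M_MED_Nautical_Pajamas player outfit, apparently for secondary hood motion.
# Each side pairs a capsule on the dyn_hood bone with a smaller capsule on its MSTR bone.
# One constraint per side joins the pair with moderate swing limits and a tighter twist limit.
# The two sides are mirrored in setup, with no constraint between them.
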PhysicsAsset /BRCosmetics/Characters/Player/Male/Medium/Bodies/M_MED_Nautical_Pajamas/Meshes/Parts/M_MED_Nautical_Pajamas_Dyn_hood
bodies {
 dyn_hood_l [{shapeType=Capsule, length=2, radius=2, center=<1 0 0>, rotation=<rot 90 0 0>}]
 dyn_hood_l_MSTR [{shapeType=Capsule, length=0.5, radius=0.5, rotation=<rot 90 0 0>}]
 dyn_hood_r [{shapeType=Capsule, length=2, radius=2, center=<1 0 0>, rotation=<rot 90 0 0>}]
 dyn_hood_r_MSTR [{shapeType=Capsule, length=0.5, radius=0.5, rotation=<rot 90 0 0>}]
}
constraints [
 {bone1Name=dyn_hood_r, bone2Name=dyn_hood_r_MSTR, swing1LimitDegrees=20, swing2LimitDegrees=20, twistLimitDegrees=5}
 {bone1Name=dyn_hood_l, bone2Name=dyn_hood_l_MSTR, swing1LimitDegrees=20, swing2LimitDegrees=20, twistLimitDegrees=5}
]
