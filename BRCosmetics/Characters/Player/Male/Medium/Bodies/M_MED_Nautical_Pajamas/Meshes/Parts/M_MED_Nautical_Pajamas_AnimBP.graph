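# Animation Blueprint for the Nautical Pajamas male medium body cosmetic, adding secondary motion to a pose copied from its attached parent mesh.
# Two chained RigidBody passes simulate separate element and hood physics assets relative to neck and head, then sprung AnimDynamics nodes move three hair bones.
# Weighted constraints make both eye master bones follow neck, head and hood bones, before a Control Rig produces the final pose.
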AnimBlueprint /BRCosmetics/Characters/Player/Male/Medium/Bodies/M_MED_Nautical_Pajamas/Meshes/Parts/M_MED_Nautical_Pajamas_AnimBP
target_skeleton=&/M_MED_Nautical_Pajamas_Skeleton
graphs [
 AnimGraph {nodes [
   n1 = RigidBody(ComponentPose=Local To Component(LocalPose=Copy Pose From Mesh() {bUseAttachedParent=true})) {
    bClampLinearTranslationLimitToRefPose=true
    bOverrideWorldGravity=true
    overridePhysicsAsset=&/M_MED_Nautical_Pajamas_Dyn_Element
    simulationSpace=BaseBoneSpace
    alphaBoolBlend {blendOption=Linear}
    baseBoneRef {boneName=neck_01}
    componentAppliedLinearAccClamp=<100 100 100>
    componentLinearAccScale=<0.2 0.2 0.2>
    componentLinearVelScale=<0.1 0.1 0.1>
    overrideWorldGravity {z=-300}
   }
   n2 = RigidBody(ComponentPose=@n1) {
    bClampLinearTranslationLimitToRefPose=true
    bOverrideWorldGravity=true
    overridePhysicsAsset=&/M_MED_Nautical_Pajamas_Dyn_hood
    simulationSpace=BaseBoneSpace
    alphaBoolBlend {blendOption=Linear}
    baseBoneRef {boneName=head}
    componentAppliedLinearAccClamp=<3000 3000 3000>
    componentLinearAccScale=<0.25 0.5 0.5>
    componentLinearVelScale=<0.3 0.3 0.3>
    overrideWorldGravity {z=-300}
   }
   n3 = Anim Dynamics - Bone: dyn_hair_a(ComponentPose=@n2, Alpha=1, ExternalForce="0,0,0") {
    angularDampingOverride=1
    angularSpringConstant=1500
    bAngularSpring=true
    bGravityOverrideInSimSpace=true
    bOverrideAngularDamping=true
    gravityScale=0.35
    simulationSpace=RootRelative
    alphaBoolBlend {blendOption=Linear}
    boundBone {boneName=dyn_hair_a}
    physicsBodyDefinitions [{
      sphereCollisionRadius=0
      boundBone {boneName=dyn_hair_a}
      boxExtents=<10 20 20>
      constraintSetup {angularLimitsMax=<10 0 10>, angularLimitsMin=<-10 0 -10>, angularTarget=<0 0 0>}
      localJointOffset=<5 0 0>
     }]
   }
   n4 = Anim Dynamics - Bone: dyn_hair_b(ComponentPose=@n3, Alpha=1, ExternalForce="0,0,0") {
    angularDampingOverride=1
    angularSpringConstant=1500
    bAngularSpring=true
    bGravityOverrideInSimSpace=true
    bOverrideAngularDamping=true
    gravityScale=0.35
    simulationSpace=RootRelative
    alphaBoolBlend {blendOption=Linear}
    boundBone {boneName=dyn_hair_b}
    physicsBodyDefinitions [{
      sphereCollisionRadius=0
      boundBone {boneName=dyn_hair_b}
      boxExtents=<10 20 20>
      constraintSetup {angularLimitsMax=<10 10 10>, angularLimitsMin=<-10 -10 -10>, angularTarget=<0 0 0>}
      localJointOffset=<10 0 0>
     }]
   }
   n5 = Anim Dynamics - Bone: dyn_hair_c(ComponentPose=@n4, Alpha=1, ExternalForce="0,0,0") {
    angularDampingOverride=1
    angularSpringConstant=1500
    bAngularSpring=true
    bGravityOverrideInSimSpace=true
    bOverrideAngularDamping=true
    gravityScale=0.35
    simulationSpace=RootRelative
    alphaBoolBlend {blendOption=Linear}
    boundBone {boneName=dyn_hair_c}
    physicsBodyDefinitions [{
      sphereCollisionRadius=0
      boundBone {boneName=dyn_hair_c}
      boxExtents=<10 20 20>
      constraintSetup {angularLimitsMax=<10 10 10>, angularLimitsMin=<-10 -10 -10>, angularTarget=<0 0 0>}
      localJointOffset=<10 0 0>
     }]
   }
   n6 = "Constraint - dyn_eyes_r_MSTR (PPP)"!(ComponentPose=@n5, Alpha=1, ConstraintWeights_0=0.0078, ConstraintWeights_1:float=0.7165, ConstraintWeights_2:float=0.2757) -> (Pose) {
    alphaBoolBlend {blendOption=Linear}
    boneToModify {boneName=dyn_eyes_r_MSTR}
    constraintSetup [{targetBone {boneName=neck_02}}, {targetBone {boneName=head}}, {targetBone {boneName=dyn_hood_r}}]
    constraintWeights [1,1,1]
   }
   n7 = "Constraint - dyn_eyes_l_MSTR (PPP)"!(ComponentPose=@n6, Alpha=1, ConstraintWeights_0=0.0078, ConstraintWeights_1:float=0.7165, ConstraintWeights_2:float=0.2757) -> (Pose) {
    alphaBoolBlend {blendOption=Linear}
    boneToModify {boneName=dyn_eyes_l_MSTR}
    constraintSetup [{targetBone {boneName=neck_02}}, {targetBone {boneName=head}}, {targetBone {boneName=dyn_hood_l}}]
    constraintWeights [1,1,1]
   }
   n8 = Control Rig(Source=Component To Local(ComponentPose=@n7)) {
    alphaBoolBlend {blendOption=Linear}
    controlRigAssetReference {blueprintRigClass=&/M_MED_Nautical_Pajamas_CtrlRig.M_MED_Nautical_Pajamas_CtrlRig_C}
   }
   Output Pose(Result=@n8) {layerGroup=None}
  ]}
 EventGraph {nodes [Event BlueprintUpdateAnimation()]}
]
variables []
