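# Physics asset for the dynamic zipper on the Nautical Pajamas male medium body cosmetic.
# Box bodies on dyn_zipper and dyn_zipper_MSTR form a short chain from spine_05, whose capsule covers the upper torso.
# The master-to-spine joint is free on all angular axes, while the zipper-to-master joint carries swing limits and a very narrow twist limit.
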PhysicsAsset /BRCosmetics/Characters/Player/Male/Medium/Bodies/M_MED_Nautical_Pajamas/Meshes/M_MED_Nautical_Pajamas_Dyn_Zipper
bodies {
 dyn_zipper [{shapeType=Box, extentX=3.5, extentY=3.5, extentZ=1.5, center=<3 0 0>, rotation=<rot 0 0 -5.0001>}]
 dyn_zipper_MSTR [{shapeType=Box, extentX=1, extentY=1, extentZ=1}]
 spine_05 [{shapeType=Capsule, length=15.0301, radius=10.1945, center=<-3.3192 3.0854 0>}]
}
constraints [
 {
  bone1Name=dyn_zipper_MSTR
  bone2Name=spine_05
  swing1LimitDegrees=45
  swing1Motion=Free
  swing2LimitDegrees=45
  swing2Motion=Free
  twistLimitDegrees=45
  twistMotion=Free
 }
 {bone1Name=dyn_zipper, bone2Name=dyn_zipper_MSTR, swing1LimitDegrees=45, swing2LimitDegrees=45, twistLimitDegrees=2}
]
